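# Unlit translucent Niagara sprite material for a sand-styled clock in the skydiving contrail effect.
# Clock progress comes from world-space height, switching to game time in frontend preview.
# It spins two T_FX_Sand channels as hands, one twelve times faster, over a face animated by curve-atlas ramps.
# Dynamic parameters drive noise distortion and hand offset; particle alpha and depth fade (skipped on mobile) shape opacity.
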Material /BRCosmetics/Athena/Effects/Materials/Skydiving/P_Contrails_Sand/M_Clock
graph {
 properties {bUsedWithNiagaraSprites=true, blendMode=BLEND_Translucent, pixelDepthOffsetMode=PDOM_Legacy, shadingModel=MSM_Unlit, usedShadingModels=Unlit}
 nodes [
  n1 = TextureObject() {samplerType=SAMPLERTYPE_LinearColor, texture=/BRCosmetics/Effects/Fort_Effects/Textures/T_FX_Sand}
  n2 = TextureCoordinate()
  n3 = DynamicParameter() {defaultValue=<rgba 1 1 0.04 1>, paramNames [Time,Random,Distort,TickOffset]}
  n4 = Add(A=Multiply(A=Time()) {constB=0.25}, B=@n3.Random)
  n5 = Subtract(A=TextureSample(UVs=Add(A=@n2, B=@n4)) {samplerType=SAMPLERTYPE_LinearColor, texture=&T_NoisePack_LinearColor}.RGB) {constB=0.5}
  n6 = Add(A=TextureCoordinate(), B=Multiply(A=ComponentMask(None=@n5) {g=true, r=true}, B=@n3.Distort) {constB=0.04})
  n7 = TextureSample(UVs=@n6, Tex=@n1) {samplerSource=SSM_Clamp_WorldGroupSettings, texture=&T_FX_Sand}
  n8 = CollectionParameter() {collection=&FortniteMaterialParameters, parameterName=AthenaFrontendPreview}
  n9 = ObjectPositionWS()
  n10 = LinearInterpolate(
   A=MaterialFunctionCall(
    Input=ComponentMask(None=@n9) {b=true}
    Input Low=Constant() {r=12000}
    Input High=ScalarParameter() {defaultValue=85000, parameterName=FullRange}
    Target Low=Constant()
    Target High=Constant() {r=1}
   ) {materialFunction=&RemapValueRange}.Result
   B=Divide(A=Time()) {constB=-24}
   Alpha=@n8
  )
  n11 = CurveAtlasRowParameter(CurveTime=Frac(None=Add(A=@n10, B=@n7.A))) {atlas=&AthenaCurveAtlas01, curve=/Game/Athena/LinearColorCurves/Curve_Sand_Ramps, defaultValue=48}
  n12 = Multiply(A=@n10) {constB=12}
  n13 = CurveAtlasRowParameter(CurveTime=Frac(None=Add(A=@n7.A, B=@n12))) {atlas=&AthenaCurveAtlas01, curve=&Curve_Sand_Ramps, defaultValue=48}
  n14 = Add(
   A=Multiply(A=@n10) {constB=25.1327}
   B=DynamicParameter() {defaultValue=<rgba 1 1 0.04 1>, paramNames [Time,Random,Distort,TickOffset]}.TickOffset
  )
  n15 = TextureSample(UVs=Rotator(Coordinate=@n6, Time=@n14), Tex=@n1) {samplerSource=SSM_Clamp_WorldGroupSettings, texture=&T_FX_Sand}
  n16 = Multiply(A=@n15.B, B=@n15.B)
  n17 = TextureSample(UVs=Rotator(Coordinate=@n6, Time=Multiply(A=@n14) {constB=12}), Tex=@n1) {samplerSource=SSM_Clamp_WorldGroupSettings, texture=&T_FX_Sand}
  n18 = Multiply(A=@n17.R, B=@n17.R)
  n19 = Add(
   A=Multiply(A=Add(A=@n11.G, B=@n13.R), B=@n7.G)
   B=Multiply(A=Add(A=@n16, B=@n18), B=VectorParameter() {parameterName=Emissive_Tick, defaultValue=<rgba 2.5 1.9581 1.3802 1>}.RGB) {constB=10}
  )
  n20 = VectorParameter() {parameterName=Emissive, defaultValue=<rgba 20 11.0774 1.5625 1>}
  MaterialFunctionCall(Emissive=Multiply(A=@n11.B, B=Multiply(A=@n19, B=@n20.RGB))) {materialFunction=&MF_FX_EmissiveAdjustments}
  n21 = Add(A=Frac(None=Multiply(A=@n10) {constB=12})) {constB=0.2}
  n22 = ScalarParameter() {defaultValue=25, parameterName=DF_FD}
  Multiply(
   A=Multiply(A=Multiply(A=@n21, B=@n19), B=ParticleColor().A) {constB=0.1}
   B=ShadingPathSwitch(Default=DepthFade(FadeDistance=@n22), Mobile=Constant() {r=1}) {desc=Migrated from FeatureLevelSwitch}
  )
 ]
}
