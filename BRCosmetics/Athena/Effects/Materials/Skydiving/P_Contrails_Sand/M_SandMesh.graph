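# Translucent, two-sided lit material for Niagara sprites, giving the sand skydiving contrail a sparkling grain look.
# Noise-distorted, screen-aligned water-noise samples supply the normal and, scrolled through time, drive curve-atlas sand colors for base color and particle-tinted emissive.
# Opacity combines a noise threshold varying along local trail UVs with a sub-UV smoke and tiled sphere mask, scaled by particle alpha.
# Swaying offset math and a red attribute set sit unwired.
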Material /BRCosmetics/Athena/Effects/Materials/Skydiving/P_Contrails_Sand/M_SandMesh
graph {
 properties {
  bEnableResponsiveAA=true
  bUseMaterialAttributes=true
  bUsedWithNiagaraSprites=true
  blendMode=BLEND_Translucent
  pixelDepthOffsetMode=PDOM_Legacy
  twoSided=true
  usedShadingModels=Default Lit
 }
 nodes [
  n1 = MaterialFunctionCall!(MaxiumURange, RangeRampHardness, MirrorTexCoordU) {materialFunction=/BRCosmetics/Weapons/FORT_Melee/Pickaxe_Venus/MF_LocalSpaceTrailUV}
  n2 = ComponentMask(None=@n1."LocalSpaceCoordinate(UV)") {r=true}
  n3 = ObjectPositionWS()
  n4 = CollectionParameter() {collection=&FortniteMaterialParameters, parameterName=AthenaFrontendPreview}
  n5 = Multiply(A=Multiply(A=@n4, B=Time())) {constB=-2500}
  n6 = ComponentMask(None=Multiply(A=Add(A=@n3, B=@n5)) {constB=-1}) {b=true}
  n7 = ScalarParameter() {defaultValue=2500, parameterName=OffsetSpeed}
  n8 = Add(A=@n2, B=Add(A=Divide(A=@n6, B=@n7) {constB=3500}, B=ScalarParameter() {parameterName=RandomOffset}))
  n9 = Multiply(A=AppendVector(A=Sine(None=@n8), B=Cosine(None=@n8)), B=ScalarParameter() {defaultValue=200, parameterName=Swipe}) {constB=100}
  n10 = OneMinus(None=@n1."LocalSpaceCoordinate(UV)")
  n11 = ComponentMask(None=Clamp(None=Subtract(A=@n10) {constB=0.2})) {r=true}
  n12 = ComponentMask(None=TextureCoordinate() {vTiling=2}) {g=true}
  n13 = Multiply(A=@n8) {constB=3}
  n14 = AppendVector(A=Cosine(None=@n13), B=Sine(None=@n13))
  n15 = Multiply(A=Multiply(A=Subtract(A=@n12), B=@n14) {constB=8}, B=ScalarParameter() {defaultValue=25, parameterName=Wide}) {constB=20}
  n16 = ComponentMask(None=@n10) {r=true}
  AppendVector(A=Add(A=Multiply(A=@n9, B=@n11), B=Multiply(A=@n15, B=@n16)), B=Constant())
  Clamp(None=@n10)
  Constant3Vector() {constant=<rgba 0.5521 0.4402 0.2134 0>}
  Constant() {r=0.8}
  n17 = TextureSample(UVs=TextureCoordinate() {uTiling=8, vTiling=8}) {texture=&T_NoisePack_01}
  n18 = Add(
   A=MaterialFunctionCall!(TextureResolution=ScalarParameter() {defaultValue=256, parameterName=Res}) {materialFunction=/Engine/Functions/Engine_MaterialFunctions02/Texturing/ScreenAlignedPixelToPixelUVs}.Result
   B=Multiply(A=ComponentMask(None=@n17.RGB) {g=true, r=true}) {constB=0.15}
  )
  n19 = TextureSample(UVs=@n18) {samplerType=SAMPLERTYPE_Normal, texture=/Game/Environments/World/Backgrounds/Locales/Temperate/Textures/T_WaterNoise}
  n20 = CollectionParameter() {
   collection=&FortniteMaterialParameters
   parameterName=SunlightDirectionalVector
  }
  DotProduct(A=@n19.R, B=@n20)
  LinearInterpolate(Alpha=SphereMask(A=WorldPosition().XYZ, B=CameraPositionWS(), Radius=Constant() {r=2000}, Hardness=Constant())) {constA=0.1, constB=0.3}
  n21 = Constant3Vector() {constant=<rgba 1 0.05 0.05 0>}
  n22 = MaterialFunctionCall() {materialFunction=&Color_Time_of_Day}
  n23 = If(A=TwoSidedSign(), "A > B"=@n22.Result, "A < B"=Multiply(A=@n22.Result) {constB=0.2}) {constB=0.5}
  n24 = Sine(None=ComponentMask(None=TextureCoordinate() {vTiling=0.5}) {g=true})
  n25 = SphereMask(A=WorldPosition().XYZ, B=CameraPositionWS(), Radius=Constant() {r=100}, Hardness=Constant())
  MakeMaterialAttributes!(
   BaseColor=@n21
   Metallic
   Specular
   Roughness
   Anisotropy
   EmissiveColor=Multiply(A=@n21, B=Multiply(A=@n23, B=@n24))
   Opacity=LinearInterpolate(B=@n24, Alpha=Multiply(A=@n25) {constB=1.5}) {constA=1}
   OpacityMask
   Normal
   Tangent
   WorldPositionOffset
   SubsurfaceColor
   ClearCoat
   ClearCoatRoughness
   AmbientOcclusion
   Refraction
   CustomizedUVs_0
   CustomizedUVs_1
   CustomizedUVs_2
   CustomizedUVs_3
   CustomizedUVs_4
   CustomizedUVs_5
   CustomizedUVs_6
   CustomizedUVs_7
   PixelDepthOffset
   ShadingModel
   Displacement
  )
  Multiply(A=@n10) {constB=5}
  n26 = TextureCoordinate() {uTiling=0.5, vTiling=0.5}
  OneMinus(None=Sine(None=ComponentMask(None=@n26) {g=true}))
  n27 = TextureSample(UVs=Add(A=Multiply(A=@n18) {constB=0.5}) {constB=0.618}) {samplerType=SAMPLERTYPE_Normal, texture=&T_WaterNoise}
  n28 = ConstantBiasScale(None=@n27.RGB)
  n29 = Time()
  n30 = MaterialFunctionCall(Float3=Frac(None=Multiply(A=@n28, B=@n29))) {materialFunction=&BreakOutFloat3Components}
  n31 = CurveAtlasRowParameter(CurveTime=@n30.R) {
   atlas=&AthenaCurveAtlas01
   curve=/Game/Athena/LinearColorCurves/Curve_Sand_EmissiveColor
   defaultValue=43
   parameterName=Emissive
  }
  n32 = Multiply(A=@n31) {constB=0.1}
  QualitySwitch(Default=@n32, Low=@n32, High=@n32, Medium=@n32)
  Reroute(None=@n2)
  ScalarParameter() {parameterName=Distance}
  n33 = ParticleColor()
  n34 = LinearInterpolate(Alpha=@n2) {constA=1.5, constB=0.6}
  n35 = OneMinus(None=Floor(None=Distance(A=@n34, B=@n27.RGB)))
  n36 = TextureSampleParameterSubUV!(UVs=TextureCoordinate(), Apply View MipBias) {parameterName=SubUV_Texture, texture=/Game/Effects/Fort_Effects/Textures/Smoke/T_SmokeSwirl_SubUV_Simple_8x8}
  n37 = SphereMask(
   A=Frac(None=TextureCoordinate() {uTiling=8, vTiling=8})
   B=Constant() {r=0.5}
   Radius=ScalarParameter() {defaultValue=0.2, parameterName=Center_SphereMask}
   Hardness=Constant()
  )
  n38 = OneMinus(None=Saturate(None=Subtract(A=@n36.R, B=@n37)))
  n39 = CurveAtlasRowParameter(CurveTime=@n30.G) {
   atlas=&AthenaCurveAtlas01
   curve=/Game/Athena/LinearColorCurves/Curve_Sand_BaseColor
   defaultValue=42
   parameterName=BaseColor
  }
  SetMaterialAttributes!(
   MaterialAttributes=SetMaterialAttributes!(MaterialAttributes, World Position Offset) {attributeSetTypes [#1]}
   Emissive Color=Multiply(A=@n32, B=@n33.RGB)
   Opacity=Multiply(A=Subtract(A=@n35, B=@n38), B=@n33.A)
   Base Color=@n39
   Normal=@n19.RGB
  ) {attributeSetTypes [
    #2
    #3
    #4
    #5
   ]}
  TextureSample() {samplerType=SAMPLERTYPE_Normal, texture=/Game/FrontEnd/Store/Cards/Assets/T_Noise_Normals}
  n40 = ComponentMask(None=VertexNormalWS()) {b=true}
  Transform(None=Normalize(VectorInput=AppendVector(A=@n40, B=@n14))) {transformSourceType=TRANSFORMSOURCE_World, transformType=TRANSFORM_Local}
 ]
}
