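# Unlit translucent material for soft mist ribbon, beam and sprite particles in the sand skydiving contrail cosmetic.
# Smoke trail U follows relative world height, switching to time panning for frontend preview or by parameter, offset by a sphere mask.
# Particle color tints emissive from the smoke normal map.
# Squared smoke, particle alpha and sphere mask form opacity, modulated by screen-aligned noise, optional end fades and depth fade, with a camera offset.
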Material /BRCosmetics/Athena/Effects/Materials/Skydiving/P_Contrails_Sand/M_SoftMist
graph {
 properties {
  bUsedWithBeamTrails=true
  bUsedWithNiagaraRibbons=true
  bUsedWithParticleSprites=true
  blendMode=BLEND_Translucent
  pixelDepthOffsetMode=PDOM_Legacy
  shadingModel=MSM_Unlit
  usedShadingModels=Unlit
 }
 nodes [
  n1 = ObjectPositionWS()
  n2 = Subtract(A=WorldPosition().XYZ, B=Multiply(A=@n1) {constB=0.9})
  n3 = Time()
  n4 = DynamicParameter() {paramNames [RandomTimeOffset, Param2, Param3, Param4]}
  n5 = CollectionParameter() {collection=&FortniteMaterialParameters, parameterName=AthenaFrontendPreview}
  n6 = LinearInterpolate(
   A=Multiply(A=ComponentMask(None=@n2) {b=true}, B=ScalarParameter() {defaultValue=0.005, parameterName=Speed})
   B=Multiply(A=Add(A=@n3, B=@n4.RandomTimeOffset), B=ScalarParameter() {defaultValue=1, parameterName=PanningSpeed})
   Alpha=Add(A=@n5, B=ScalarParameter() {parameterName=UsePanningTexture})
  ) {constAlpha=1}
  n7 = SphereMask(
   A=WorldPosition().XYZ
   B=Add(A=@n1, B=VectorParameter() {parameterName=InitialOffset, defaultValue=<rgba 0 0 0 1>}.RGB)
   Radius=ScalarParameter() {defaultValue=400, parameterName=SphereRadius}
   Hardness=Constant()
  )
  n8 = AppendVector(
   A=Multiply(A=Add(A=@n6, B=@n7), B=ScalarParameter() {defaultValue=0.5, parameterName=UScale})
   B=ComponentMask(None=TextureCoordinate()) {g=true}
  )
  n9 = TextureSample(UVs=@n8) {samplerType=SAMPLERTYPE_Normal, texture=/Game/Effects/Fort_Effects/Textures/Smoke/T_Smoke_Trail_Soft_N}
  n10 = Multiply(
   A=AppendVector(A=ComponentMask(None=@n9.RGB) {g=true, r=true}, B=Constant())
   B=ScalarParameter() {defaultValue=0.35, parameterName=Colorize}
  )
  n11 = ParticleColor()
  MaterialFunctionCall(Emissive=Add(A=Multiply(A=@n10, B=@n11.RGB), B=Multiply(A=@n9.A, B=@n11.RGB))) {materialFunction=&MF_FX_EmissiveAdjustments}
  n12 = ScalarParameter() {defaultValue=25, parameterName=CameraOffset}
  MaterialFunctionCall(Offset Amount=@n12) {materialFunction=/Engine/Functions/Engine_MaterialFunctions02/WorldPositionOffset/CameraOffset}
  n13 = Multiply(
   A=TextureSample(UVs=@n8) {samplerType=SAMPLERTYPE_Grayscale, texture=&T_Smoke_Trail_Soft}.R
   B=Multiply(A=@n11.A, B=@n7)
  )
  n14 = Multiply(A=@n13, B=@n13)
  n15 = ScalarParameter() {defaultValue=1024, parameterName=Res}
  n16 = TextureCoordinate() {uTiling=0.2, vTiling=2}
  n17 = ComponentMask(None=TextureSample(UVs=Panner(Coordinate=@n16) {speedX=0.5, speedY=0.2}) {texture=&T_NoisePack_01}.RGB) {g=true, r=true}
  n18 = Multiply(A=Add(
    A=MaterialFunctionCall!(TextureResolution=@n15) {materialFunction=/Engine/Functions/Engine_MaterialFunctions02/Texturing/ScreenAlignedPixelToPixelUVs}.Result
    B=Multiply(A=@n17) {constB=0.1}
   )) {constB=0.5}
  n19 = Distance(
   A=LinearInterpolate(Alpha=@n7) {constA=1, constB=0}
   B=TextureSample(UVs=Add(A=@n18) {constB=0.618}) {samplerType=SAMPLERTYPE_Normal, texture=/Game/Environments/World/Backgrounds/Locales/Temperate/Textures/T_WaterNoise}.RGB
  )
  n20 = LinearInterpolate(Alpha=Saturate(None=OneMinus(None=@n19))) {constA=0.75, constB=1.25}
  n21 = TextureCoordinate()
  n22 = Multiply(A=OneMinus(None=ComponentMask(None=@n21) {r=true}), B=ScalarParameter() {defaultValue=12, parameterName=SoftEnd}) {constA=20, constB=12}
  n23 = TextureCoordinate()
  n24 = Saturate(None=Multiply(A=ComponentMask(None=@n23) {r=true}, B=ScalarParameter() {defaultValue=12, parameterName=SoftStart}) {constA=20, constB=12})
  n25 = LinearInterpolate(B=Multiply(A=Saturate(None=@n22), B=@n24), Alpha=ScalarParameter() {parameterName=UseSoftEnd}) {constA=1}
  n26 = ScalarParameter() {parameterName=UseSoftStart}
  Multiply(A=Multiply(A=Multiply(A=@n14, B=@n20), B=LinearInterpolate(A=@n25, B=@n24, Alpha=@n26) {constA=1}), B=DepthFade(FadeDistance=@n12))
 ]
}
