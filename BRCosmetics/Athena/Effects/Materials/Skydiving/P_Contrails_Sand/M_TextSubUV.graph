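# Unlit translucent Niagara sprite material for the sand skydiving contrail, showing one of twelve cells from a 4x4 clock-text atlas.
# World height of the object, remapped up to FullRange and shifted by RangeOffset, picks the cell.
# The AthenaFrontendPreview collection parameter blends that driver to time, so previews cycle.
# Particle color tints the red channel plus boosted green glow; red also sets faint opacity.
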Material /BRCosmetics/Athena/Effects/Materials/Skydiving/P_Contrails_Sand/M_TextSubUV
graph {
 properties {bUsedWithNiagaraSprites=true, blendMode=BLEND_Translucent, pixelDepthOffsetMode=PDOM_Legacy, shadingModel=MSM_Unlit, usedShadingModels=Unlit}
 nodes [
  n1 = ParticleColor()
  n2 = CollectionParameter() {collection=&FortniteMaterialParameters, parameterName=AthenaFrontendPreview}
  n3 = ObjectPositionWS()
  n4 = LinearInterpolate(
   A=MaterialFunctionCall(
    Input=ComponentMask(None=@n3) {b=true}
    Input Low=Constant() {r=12000}
    Input High=ScalarParameter() {defaultValue=75000, parameterName=FullRange}
    Target Low=Constant()
    Target High=Constant() {r=1}
   ) {materialFunction=&RemapValueRange}.Result
   B=Divide(A=Time()) {constB=24}
   Alpha=@n2
  )
  n5 = ScalarParameter() {defaultValue=-0.0833, parameterName=RangeOffset}
  n6 = Multiply(A=Frac(None=Add(A=@n4, B=@n5))) {constB=12}
  n7 = AppendVector(A=Floor(None=@n6), B=Floor(None=Divide(A=@n6) {constB=4}))
  n8 = TextureSample(UVs=Add(A=Multiply(A=@n7) {constB=0.25}, B=TextureCoordinate() {uTiling=0.25, vTiling=0.25})) {texture=/BRCosmetics/Effects/Fort_Effects/Textures/T_ClockTextSubUV}
  n9 = Multiply(A=@n8.G) {constB=25}
  MaterialFunctionCall(Emissive=Multiply(A=@n1.RGB, B=Add(A=@n8.R, B=@n9))) {materialFunction=&MF_FX_EmissiveAdjustments}
  Multiply(A=@n1.A, B=Multiply(A=@n8.R) {constB=0.15})
 ]
}
