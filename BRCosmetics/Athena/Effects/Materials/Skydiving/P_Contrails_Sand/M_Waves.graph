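# Translucent unlit particle material for the sand skydiving contrail, drawing swirling, distorted radial wave shapes.
# A sphere mask twists time-driven UV rotation, and panning noise, offset and scaled by dynamic parameters, warps those UVs into a second radial mask.
# That mask samples a sand-ramp curve atlas row, whose alpha, multiplied by the mask's sine, drives emissive and opacity with particle color.
# Depth fade softens opacity, except on mobile.
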Material /BRCosmetics/Athena/Effects/Materials/Skydiving/P_Contrails_Sand/M_Waves
graph {
 properties {
  bUsedWithNiagaraSprites=true
  bUsedWithParticleSprites=true
  blendMode=BLEND_Translucent
  pixelDepthOffsetMode=PDOM_Legacy
  shadingModel=MSM_Unlit
  usedShadingModels=Unlit
 }
 nodes [
  n1 = TextureCoordinate()
  n2 = Constant() {r=0.5}
  n3 = SphereMask(A=TextureCoordinate(), B=@n2, Radius=@n2, Hardness=Constant())
  n4 = Multiply(A=Add(A=Multiply(A=@n3) {constB=4}, B=Time()), B=ScalarParameter() {defaultValue=12, parameterName=Speed_Rotation})
  n5 = Time()
  n6 = ScalarParameter() {defaultValue=0.25, parameterName=PannerSpeed}
  n7 = DynamicParameter() {defaultValue=<rgba 1 1 0 1>, paramNames [Time,Random,Distort,Tick_Offset]}
  n8 = Add(A=TextureCoordinate(), B=Add(A=Multiply(A=@n5, B=@n6) {constB=0.25}, B=@n7.Random))
  n9 = ComponentMask(None=Subtract(A=TextureSample(UVs=@n8) {samplerType=SAMPLERTYPE_LinearColor, texture=&T_NoisePack_LinearColor}.RGB) {constB=0.5}) {g=true, r=true}
  n10 = Constant() {r=0.5}
  n11 = SphereMask(
   A=Add(A=Rotator(Coordinate=@n1, Time=@n4) {centerX=0.48, centerY=0.48}, B=Multiply(A=@n9, B=@n7.Distort) {constB=0.04})
   B=@n10
   Radius=@n10
   Hardness=Constant()
  )
  n12 = Sine(None=Multiply(A=@n11) {constB=0.5})
  n13 = LinearInterpolate(A=@n11, B=@n12) {constAlpha=0.15}
  n14 = Multiply(A=Time(), B=ScalarParameter() {parameterName=Speed})
  Frac(None=Multiply(A=Add(A=@n13, B=@n14)) {constB=2.8})
  n15 = ParticleColor()
  n16 = CurveAtlasRowParameter(CurveTime=@n11) {atlas=&AthenaCurveAtlas01, curve=/Game/Athena/LinearColorCurves/Curve_Sand_Ramps, defaultValue=48}
  n17 = Multiply(A=@n16.A, B=@n12)
  MaterialFunctionCall(Emissive=Multiply(A=@n15.RGB, B=@n17)) {materialFunction=&MF_FX_EmissiveAdjustments}
  n18 = Multiply(A=@n15.A, B=Multiply(A=@n17) {constB=0.1})
  n19 = ScalarParameter() {defaultValue=2, parameterName=DF_FD}
  ShadingPathSwitch(Default=Multiply(A=@n18, B=DepthFade(FadeDistance=@n19)), Mobile=@n18) {desc=Migrated from FeatureLevelSwitch}
 ]
}
